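; BP_HeatmapRenderer_BoxElement is a heatmap element that shows one data point as a box whose height follows its severity.
; DrawElement calls the parent, then reads the severity at its data index from the current data object.
; It normalizes that value against the data's severity range, with a floor applied to the minimum.
; The result drives the box's Z scale and Range variable, while the owning manager's grid size sets the X/Y footprint.
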(event EventBeginPlay
  (|CallParentFunction) ; This node is disabled and will not be called.\nDrag off pins to build functionality.
)

(event Collision|EventActorBeginOverlap
  (params (OtherActor :type Actor%20Object%20Reference))
  (|CallParentFunction OtherActor) ; This node is disabled and will not be called.\nDrag off pins to build functionality.
)

(event EventTick
  (params (DeltaSeconds :type Float%20%28single-precision%29))
  (|CallParentFunction DeltaSeconds) ; This node is disabled and will not be called.\nDrag off pins to build functionality.
)

(fn DrawElement
  (bind _dataindex (Variables|Default|GetDataIndex))
  (bind _currentdataobject (Variables|Default|GetCurrentDataObject))
  (bind _severities (Variables|/Game/Blueprints/HeatmapRenderer/Data/HeatmapRendererDataObject.HeatmapRendererDataObject_C|GetSeverities _currentdataobject))
  (bind _output (Utilities|Array|Get(aref) _severities _dataindex))
  (|CallParentFunction)
  (bind (_severitymin _severitymax) (Class|HeatmapRendererDataObject|GetSeverityRange (Variables|Default|GetCurrentDataObject) false))
  (Transformation|SetWorldScale3D (Variables|Default|GetBoxMesh) (Math|Vector|MakeVector (Variables|/Game/Blueprints/HeatmapRenderer/BP_HeatmapRenderer_Manager.BP_HeatmapRenderer_Manager_C|GetGridSize (Variables|Default|GetOwningManager)) (Variables|/Game/Blueprints/HeatmapRenderer/BP_HeatmapRenderer_Manager.BP_HeatmapRenderer_Manager_C|GetGridSize (Variables|Default|GetOwningManager)) (* 5.5e+04 (Math|Float|NormalizetoRange _output (Math|Float|Max(Float) _severitymin 1e+02) _severitymax))))
  (Variables|Default|SetRange (Math|Float|NormalizetoRange _output (Math|Float|Max(Float) _severitymin 1e+02) _severitymax)))

(fn ConstructionScript
  (|CallParentFunction))
