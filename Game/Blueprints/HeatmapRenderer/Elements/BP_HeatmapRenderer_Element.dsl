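; BP_HeatmapRenderer_Element is a single element of the heatmap renderer, tied to one entry in a heatmap data object.
; DrawElement reads the Severities array from the current HeatmapRendererDataObject at the element's DataIndex and stores that value in SeverityDisplay.
; BeginPlay, overlap, Tick and the construction script are empty, and the file shows no visual update beyond setting that variable.
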(event EventBeginPlay)

(event Collision|EventActorBeginOverlap
  (params (OtherActor :type Actor%20Object%20Reference)))

(event EventTick
  (params (DeltaSeconds :type Float%20%28single-precision%29)))

(fn DrawElement
  (Variables|Default|SetSeverityDisplay (Utilities|Array|Get(aref) (Variables|/Game/Blueprints/HeatmapRenderer/Data/HeatmapRendererDataObject.HeatmapRendererDataObject_C|GetSeverities (Variables|Default|GetCurrentDataObject)) (Variables|Default|GetDataIndex))))

(fn ConstructionScript)
